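# Physics asset for the Desert Ops Camo backpack cosmetic, defining simulation bodies for its dangling attachments.
# A large box on the pack bone anchors simple capsule, box and sphere bodies for zippers, belt straps, light, dagger, grenade and a hanging piece.
# Each attachment constrains directly to the pack in a star layout, with angular limits tuned per part.
# The dagger locks twist and one swing axis, leaving a narrow single-axis swing.
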PhysicsAsset /BRCosmetics/Accessories/FORT_Backpacks/Backpack_F_MED_Desert_Ops_Camo/Meshes/F_MED_Desert_Ops_Camo_dyn
bodies {
 dyn_belt_traps_l_01 [{shapeType=Capsule, length=1, radius=1, center=<3 0 0>, rotation=<rot 0 0 90>}]
 dyn_belt_traps_l_02 [{shapeType=Capsule, length=1, radius=1, center=<3 0 0>, rotation=<rot 0 0 90>}]
 dyn_belt_traps_r_01 [{shapeType=Capsule, length=1, radius=1, center=<3 0 0>, rotation=<rot 0 0 90>}]
 dyn_belt_traps_r_02 [{shapeType=Capsule, length=1, radius=1, center=<2 0 0>, rotation=<rot 0 0 90>}]
 dyn_dagger [{shapeType=Box, extentX=15, extentY=7.5, extentZ=3, center=<-1.3 0 1>}]
 dyn_grenade [{shapeType=Sphere, radius=3.4, center=<5 0.25 -0.15>}]
 dyn_hanging [{shapeType=Box, extentX=3, extentY=2, extentZ=0.5, center=<2 0 0>}]
 dyn_light [{shapeType=Capsule, length=4, radius=1.2, center=<-0.5 0 1>, rotation=<rot 90 0 0>}]
 dyn_zipper_01 [{shapeType=Capsule, length=1, radius=0.6, center=<1.5 0 0>, rotation=<rot 0 0 90>}]
 dyn_zipper_02 [{shapeType=Capsule, length=1, radius=0.6, center=<1.5 0 0>, rotation=<rot 0 0 90>}]
 dyn_zipper_03 [{shapeType=Capsule, length=1, radius=0.6, center=<2 0 0>, rotation=<rot 0 0 90>}]
 pack [{shapeType=Box, extentX=20.3503, extentY=14.7011, extentZ=25.6403, center=<-0.3348 -7.3147 -1.6708>}]
}
constraints [
 {bone1Name=dyn_zipper_01, bone2Name=pack, swing1LimitDegrees=45, swing2LimitDegrees=45, twistLimitDegrees=5}
 {bone1Name=dyn_zipper_02, bone2Name=pack, swing1LimitDegrees=45, swing2LimitDegrees=45, twistLimitDegrees=5}
 {bone1Name=dyn_zipper_03, bone2Name=pack, swing1LimitDegrees=45, swing2LimitDegrees=45, twistLimitDegrees=5}
 {bone1Name=dyn_belt_traps_r_01, bone2Name=pack, swing1LimitDegrees=10, swing2LimitDegrees=22.5, twistLimitDegrees=5}
 {bone1Name=dyn_belt_traps_r_02, bone2Name=pack, swing1LimitDegrees=10, swing2LimitDegrees=22.5, twistLimitDegrees=5}
 {bone1Name=dyn_belt_traps_l_01, bone2Name=pack, swing1LimitDegrees=10, swing2LimitDegrees=22.5, twistLimitDegrees=5}
 {bone1Name=dyn_belt_traps_l_02, bone2Name=pack, swing1LimitDegrees=10, swing2LimitDegrees=22.5, twistLimitDegrees=5}
 {bone1Name=dyn_light, bone2Name=pack, swing1LimitDegrees=10, swing2LimitDegrees=2, twistLimitDegrees=5}
 {
  bone1Name=dyn_dagger
  bone2Name=pack
  swing1LimitDegrees=5
  swing2LimitDegrees=45
  swing2Motion=Locked
  twistLimitDegrees=45
  twistMotion=Locked
 }
 {bone1Name=dyn_hanging, bone2Name=pack, swing1LimitDegrees=5, swing2LimitDegrees=45, twistLimitDegrees=5}
 {bone1Name=dyn_grenade, bone2Name=pack, swing1LimitDegrees=45, swing2LimitDegrees=45, twistLimitDegrees=5}
]
